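; Gameplay ability for the DentureSay air dash, a forward dash into an air-swing attack, extending a parent ability.
; A hotfix can hold the leap until a montage notify.
; Continuous targeting starts a tick after the dash movement begins; a hit, delay or movement end ends the dash, jumping the montage to "AirSwing".
; The impact notify allows air combos and selects damage targets, deduplicates them by actor, then applies effects next tick.
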(event Ability|EventActivateAbility
  (bind _hotfix_delayleap (Variables|Hotfix|GetHotfix_DelayLeap))
  (bind _returnvalue (ScalableFloat|GetValueasBool _hotfix_delayleap))
  ;[0.04741 0.276 0.08513] ActivateAbility
  (Variables|Transient|SetbAllowLeap (not _returnvalue)) ; Ignore leap that occurs during activation \nif delayed leap is enabled
  (Variables|PowerHover|SetHoverOnActivate _returnvalue)
  (|CallParentFunction)
  (FortMovement|CancelJumpAbility (Variables|Default|GetPlayerPawn) false) ; For longer DentureSay jump, \nupwards physics may continue to apply \nafter quick dash if not canceled
)

(event EventLeap
  ;[0.15 0.03721 0.03291 0.5] Override Leap to filter filter calls
  (Variables|Transient|GetbAllowLeap
    (:then
      (|CallParentFunction))
    (:else)))

(event EventOnMMEStart
  ;[0.06319 0.3281 0.2339 0.5] Handle MME events
  (|CallParentFunction)
  (bind _returnvalue (Utilities|Time|SetTimerforNextTickbyEvent (EventDispatchers|CreateEvent :!event Custom|BeginTargeting))) ; Optimization to prevent targeting traces from running on the same frame that the MME begins.
  (Variables|Transient|SetBeginTargetingFrameDelayTimerHandle _returnvalue)
  (bind _asynctaskproxy (GameplayTasks|TaskWaitDelay 0 (ScalableFloat|GetValueatLevel (Variables|Hotfix|GetHotfix_DashDurationBeforeMontageUpdate)))
    (:then
      (Variables|Transient|AsyncTask|SetAsyncTask_DelayDashMontageUpdate _asynctaskproxy))
    (:OnFinish
      (CallFunction|EndTargeting)
      (CallFunction|EndDash))))

(event EventOnMMEEnd
  ;[0.06319 0.3281 0.2339 0.5] Handle MME events
  (|CallParentFunction)
  (CallFunction|EndTargeting)
  (CallFunction|EndDash))

(event EventOnMMETimeOut
  ;[0.06319 0.3281 0.2339 0.5] Handle MME events
  (|CallParentFunction)
  (CallFunction|EndTargeting)
  (CallFunction|EndDash))

(event Custom|BeginTargeting
  ;[0.1062 0.01917 0.1094] ForwardTargeting: Look for potential targets in front of the player.
  (bind (_asynctaskproxy _targetdata) (Ability|Tasks|DoTargetSelectionContinuous "ContinuousTargeting" "(TagName=\"Gameplay.Application.Hit\")")
    (:then
      (bind _output_get (Variables|Transient|AsyncTask|SetAsyncTask_ContinuousTargeting _asynctaskproxy)))
    (:Targeted
      (if (Utilities|Array|IsNotEmpty (Ability|TargetData|GetAllActorsfromTargetData _targetdata))
        (GameplayTasks|EndTask _output_get)
        (Variables|Transient|SetContinuousTargetingHitResult (Ability|TargetData|GetHitResultfromTargetData _targetdata))
        (CallFunction|EndDash)))))

(event Custom|EndTargeting
  ;[0.1062 0.01917 0.1094] ForwardTargeting: Look for potential targets in front of the player.
  (GameplayTasks|EndTask (Variables|Transient|AsyncTask|GetAsyncTask_ContinuousTargeting))
  (Utilities|Time|ClearandInvalidateTimerbyHandle (Variables|Transient|GetBeginTargetingFrameDelayTimerHandle)))

(event EventEndDash
  ;[0.03101 0.1119 0.1458] EndDash: The dash movement has ended either due to finding an target along the path or due to the dash MME ending.
  ; Jump to the next montage section.
  (Utilities|FlowControl|Sequence
    (:then_0
      (if (not (Variables|Transient|GetbDashEnded))
        (Ability|Animation|MontageJumptoSection "AirSwing")))
    (:then_1
      (|CallParentFunction))))

(event EventCastMontageNotifyStart
  (params (NotifyTag :type Gameplay%20Tag%20Structure))
  ;[0.01664 0.1007 0.1094] NotifyStart: When the impact notification is received, perform the damage execution.
  (|CallParentFunction NotifyTag)
  (if (== NotifyTag (Variables|Settings|GetAnimNotify_Leap))
    (Variables|Transient|GetbAllowLeap
      (:then)
      (:else
        (Variables|Transient|SetbAllowLeap true)
        (CallFunction|Leap)))
    (elif (== NotifyTag (Variables|Settings|GetAnimNotify_Impact))
      (Utilities|FlowControl|Sequence
        (:then_0
          ;;[0.15 0.04292 0.03575 0.5] Stop moving towards target
          (if (not (Variables|Transient|GetbUsedMoveToActorMME))
            (bind _outfortplayerpawn (CallFunction|GetAvatarasValidFortPlayerPawn)
              (:ValidFortPlayerPawn
                (bind _movementmoderuntimedata (Variables|Transient|GetMovementModeRuntimeData)
                  (:then
                    (Class|FortMovementModeExtBlueprintLibrary|EndMovementModeExtensiononActor _movementmoderuntimedata _outfortplayerpawn))
                  (:else))))))
        (:then_1
          ;;[0.03591 0.15 0.05327 0.5] Allow mantis air combo
          (Ability|ApplyGameplayEffectToOwner (Variables|Settings|GetGE_AllowAirCombo)))
        (:then_2
          ;;[0.15 0.02805 0.03809 0.5] Performs the target selection that eventually applies the damage of the attack.
          (bind (_asynctaskproxy _targetdata _applicationtag) (Ability|Tasks|DoTargetSelection "DamageTargetSelection" (Variables|Settings|GetAnimNotify_Impact))
            (:then
              (Variables|Transient|AsyncTask|SetAsyncTask_DamageTargetSelection _asynctaskproxy))
            (:Targeted
              (bind _filteredtargetdata (Ability|FilterTargetData _targetdata))
              (Variables|Transient|SetCachedDamageTargetDataHandle _filteredtargetdata)
              (Utilities|FlowControl|DelayUntilNextTick) ; Perf
              (Ability|ApplyGameplayEffectContainer (Variables|Transient|GetCachedDamageTargetDataHandle) _applicationtag))))))))

(event Ability|EventOnEndAbility
  (params (bWasCancelled :type Boolean))
  ;[0.08333 0.0335 0.007236] OnEndAbility: Clean up.
  (Utilities|FlowControl|Sequence
    (:then_0
      (GameplayTasks|EndTask (Variables|Transient|AsyncTask|GetAsyncTask_ContinuousTargeting))
      (CallFunction|EndTargeting))
    (:then_1
      (|CallParentFunction bWasCancelled))))

(fn FilterTargetData
  (params (TargetData :type Gameplay%20Ability%20Target%20Data%20Handle%20Structure))
  ; FilterTargetData: Removing duplicate actor entries.
  (for _index (range (Ability|TargetData|GetDataCountfromTargetData TargetData))
    (Utilities|IsValid (.HitActor (Collision|BreakHitResult (Ability|TargetData|GetHitResultfromTargetData TargetData _index)))
      (:Is%20Valid
        (bind _returnvalue (Utilities|Array|AddUnique (LocalVariables|GetHitActors) (.HitActor (Collision|BreakHitResult (Ability|TargetData|GetHitResultfromTargetData TargetData _index)))))
        (if (>= _returnvalue 0)
          (bind _returnvalue_2 (Ability|TargetData|AppendTargetDataHandle (LocalVariables|GetReturnHandle) (Ability|TargetData|AbilityTargetDatafromHitResult (Ability|TargetData|GetHitResultfromTargetData TargetData _index))))
          (LocalVariables|SetReturnHandle _returnvalue_2)))))
  (return (LocalVariables|GetReturnHandle)))

(fn GetCustomAbilitySourceTransform
  (bind _returnvalue (CallFunction|GetAvatarasFortPlayerPawn))
  (bind (_returnvalue_location _returnvalue_rotation _returnvalue_scale) (Math|Transform|BreakTransform (Transformation|GetActorTransform _returnvalue)))
  ;[0.1044 0.01678 0.1146] GetCustomAbilitySourceTransform: This function is used to return the source transform for the gameplay effect containerof 'Gameplay.Application.hit', which is used during continuous targeting while the dash takes place.
  ; IF the player is moving, this returns the velocity direction as the rotation.
  (Utilities|IsValid _returnvalue
    (:Is%20Valid
      (return _returnvalue_location (Math|Rotator|SelectRotator (Math|Rotator|MakeRotfromX (Transformation|GetVelocity _returnvalue)) (Transformation|GetActorRotation _returnvalue) (> (Math|Vector|VectorLength (Transformation|GetVelocity _returnvalue)) 0)) _returnvalue_scale))
    (:Is%20Not%20Valid
      ;;[1 0 0] Warning: Avatar player pawn is NULL.
      (Fortnite|FortShippingLog self "GetCustomAbilitySourceTransform: Avatar player pawn is NULL." true)
      (return "0, 0, 0" "0, 0, 0" "1.000000,1.000000,1.000000"))))

(fn GetTargetingHit
  (bind _return_value (Utilities|GetTargetingComponent)
    (:Is%20Valid
      (return (Class|BAbilityTargetingFeedbackComponentParent|GetLatestTargetingTraceHitResult _return_value)))
    (:Is%20Not%20Valid
      (return (|CallParentFunction)))))

(fn OverrideFailedReason
  (params (FailedReason :type Gameplay%20Tag%20Container%20Structure%20%28by%20ref%29))
  ;[0.15 0.01172 0.01172 0.5] Only show error message for unexpected failure reasons
  ; Ignore if blocked by things like melee combo
  (if (GameplayTags|HasAnyTags FailedReason (Variables|Settings|GetExpectedFailureReasons))
    (return "(GameplayTags=)")
    (else
      (return "(GameplayTags=((TagName=\"Athena.Ability.Failed.NoMessage\")))"))))

(fn RemoveCameraMode
  ;[0.009367 0.09587 0.1042] RemoveCameraMode: Remove the applied camera mode.
  (Utilities|FlowControl|Sequence
    (:then_0
      (bind _outfortplayerpawn (CallFunction|GetAvatarasValidFortPlayerPawn)
        (:ValidFortPlayerPawn
          (Utilities|IsValid (Fortnite|GetFortPlayerControllerfromActor _outfortplayerpawn)
            (:Is%20Valid
              (bind _playercameramanager (Variables|/Script/Engine.PlayerController|GetPlayerCameraManager (Fortnite|GetFortPlayerControllerfromActor _outfortplayerpawn))
                (:then
                  (bind _asfort_player_camera_base (Utilities|Casting|CastToFortPlayerCameraBase _playercameramanager))
                  (Camera|RemoveCameraModeOverride _asfort_player_camera_base (Variables|Camera|GetCameraModeTag) 0 true))
                (:else)))))))
    (:then_1
      (bind _cameramodifier (Variables|Camera|GetCameraModifier)
        (:then
          (bind _asfort_camera_modifier_denture_say (Utilities|Casting|CastToFortCameraModifier_DentureSay _cameramodifier))
          (CameraModifier|WindDownandRemove _asfort_camera_modifier_denture_say :!class /Script/DentureSayRuntime.FortCameraModifier_DentureSay))
        (:else)))))

(fn ResetTransientState
  ;[0.08847 0.15 0.03889 0.5] Crucial functionality to keep up to date. Resets all the transient variables to their default values so that the GA can be executed correctly when invoked multiple times per instance.
  (|CallParentFunction)
  (Variables|Transient|SetContinuousTargetingHitResult)
  (Variables|Transient|SetCachedDamageTargetDataHandle))
