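; Gameplay ability for DentureSay that lets a held Primary input carry through an air dash into an air combo.
; On activation it calls the parent, then gets the avatar pawn's Mantis pawn component.
; If that component is valid and Primary is still pressed, it simulates a Primary Pressed event to attempt to start the combo.
; The end-ability event holds only a disabled parent-call placeholder.
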(event Ability|EventActivateAbility
  (|CallParentFunction)
  (bind _asfort_pawn (Utilities|Casting|CastToFortPawn (Ability|GetAvatarActorfromActorInfo)))
  ; Attempt to start combo if input is being held. Allows for transitioning to combo by holding input through air dash
  (Utilities|IsValid (Actor|GetComponentbyClass _asfort_pawn "/Script/MantisRuntime.FortMantisPawnComponent")
    (:Is%20Valid
      (if (Mantis|IsSpecificInputPressed (Actor|GetComponentbyClass _asfort_pawn "/Script/MantisRuntime.FortMantisPawnComponent") "Primary")
        (Mantis|SimulateInputEvent (Actor|GetComponentbyClass _asfort_pawn "/Script/MantisRuntime.FortMantisPawnComponent") "Primary" "Pressed" true true)))))

(event Ability|EventOnEndAbility
  (params (bWasCancelled :type Boolean))
  (|CallParentFunction bWasCancelled) ; This node is disabled and will not be called.\nDrag off pins to build functionality.
)
